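# HUD widget for riding creatures, showing the mount's icon, sprint energy bar and health bar with numeric value.
# PlayerRiderComponent events, rebound on pawn change, show or hide it and drive material Progress fills with timer-driven delta bars and low-energy flashing.
# Alternating async gameplay tag waits on the player pawn toggle boost glow and sparks, positioned by energy.
# Touch input gets resized bars; mobile touch also force-refreshes; otherwise overshield visibility shifts the widget vertically.
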WidgetBlueprint /RidingContent/UI/RiderWidget
animations [
 A_FullBarAnim {
  binding_count=1
  length_seconds=0.75
  object=&self:A_FullBarAnim
  sequence {bindings [Overlay_Stats {
     id=#1
     possessed_class=Overlay
     tracks [{
       type=MovieSceneFloatTrack
       display_name=Render Opacity
       sections [{type=MovieSceneFloatSection, end_frame=15, start_frame=0, channels [None {type=MovieSceneScriptingFloatChannel, keys [[0,1],[10,1],[15,0]]}]}]
      }]
    }]}
 }
 A_LowAnim {
  binding_count=1
  length_seconds=0.5
  object=&self:A_LowAnim
  sequence {bindings [Overlay_Stats {
     id=#2
     possessed_class=Overlay
     tracks [{
       type=MovieSceneFloatTrack
       display_name=Render Opacity
       sections [{
         type=MovieSceneFloatSection
         end_frame=10
         start_frame=0
         channels [None {type=MovieSceneScriptingFloatChannel, keys [[0,1],[5,0.33],[10,1]]}]
        }]
      }]
    }]}
 }
 A_BoostStartedAnim {
  binding_count=3
  length_seconds=0.4
  object=&self:A_BoostStartedAnim
  sequence {
   bindings [
    Overlay_BoostSparks {
     id=#3
     possessed_class=Overlay
     tracks [{
       type=MovieSceneFloatTrack
       display_name=Render Opacity
       sections [{type=MovieSceneFloatSection, end_frame=24, start_frame=0, channels [None {type=MovieSceneScriptingFloatChannel, keys [[0,0],[24,1]]}]}]
      }]
    }
    BoostGlowBorder {
     id=#4
     possessed_class=Border
     tracks [{
       type=MovieSceneFloatTrack
       display_name=Render Opacity
       sections [{type=MovieSceneFloatSection, end_frame=24, start_frame=0, channels [None {type=MovieSceneScriptingFloatChannel, keys [[0,0],[24,1]]}]}]
      }]
    }
    FillBar {
     id=#5
     possessed_class=Image
     tracks [{
       type=MovieSceneWidgetMaterialTrack
       sections [{
         type=MovieSceneParameterSection
         end_frame=24
         start_frame=0
         channels [
          Progress_Color_StopR {type=MovieSceneScriptingFloatChannel, const [1,2]}
          Progress_Color_StopG {type=MovieSceneScriptingFloatChannel, const [1,2]}
          Progress_Color_StopB {type=MovieSceneScriptingFloatChannel, keys [[0,0],[24,1]]}
          Progress_Color_StopA {type=MovieSceneScriptingFloatChannel, const [1,2]}
         ]
        }]
      }]
    }
   ]
   display_rate {numerator=60}
  }
 }
 A_BoostCompletedAnim {
  binding_count=3
  length_seconds=0.3
  object=&self:A_BoostCompletedAnim
  sequence {
   bindings [
    BoostGlowBorder {
     id=#6
     possessed_class=Border
     tracks [{
       type=MovieSceneFloatTrack
       display_name=Render Opacity
       sections [{type=MovieSceneFloatSection, end_frame=18, start_frame=0, channels [None {type=MovieSceneScriptingFloatChannel, keys [[0,1],[18,0]]}]}]
      }]
    }
    FillBar {
     id=#7
     possessed_class=Image
     tracks [{
       type=MovieSceneWidgetMaterialTrack
       sections [{
         type=MovieSceneParameterSection
         end_frame=18
         start_frame=0
         channels [
          Progress_Color_StopR {type=MovieSceneScriptingFloatChannel, const [1,2]}
          Progress_Color_StopG {type=MovieSceneScriptingFloatChannel, const [1,2]}
          Progress_Color_StopB {type=MovieSceneScriptingFloatChannel, keys [[0,1],[18,0]]}
          Progress_Color_StopA {type=MovieSceneScriptingFloatChannel, const [1,2]}
         ]
        }]
      }]
    }
    Overlay_BoostSparks {
     id=#8
     possessed_class=Overlay
     tracks [{
       type=MovieSceneFloatTrack
       display_name=Render Opacity
       sections [{type=MovieSceneFloatSection, end_frame=18, start_frame=0, channels [None {type=MovieSceneScriptingFloatChannel, keys [[0,1],[18,0]]}]}]
      }]
    }
   ]
   display_rate {numerator=60}
  }
 }
]
graphs [
 ClearOldBindings {nodes [
   ClearOldBindings() {is=title K2Node_FunctionEntry}
   > n1 = Get!(execute:Exec) -> (then:Exec, else:Exec, PlayerRiderComponentRef:Player Rider Component Object Reference)
   > Unbind Event from Rider UIEnd(self:Player Rider Component Object Reference=@n1, Delegate=Create Event()) {is=title K2Node_RemoveDelegate}
   > Unbind Event from Rider UINew Stats(self:Player Rider Component Object Reference=@n1, Delegate=Create Event()) {is=title K2Node_RemoveDelegate}
   > Unbind Event from Rider UIUpdate Energy(self:Player Rider Component Object Reference=@n1, Delegate=Create Event()) {is=title K2Node_RemoveDelegate}
   > Unbind Event from Rider UIUpdate Health(self:Player Rider Component Object Reference=@n1, Delegate=Create Event()) {is=title K2Node_RemoveDelegate}
   > Set PlayerRiderComponentRef(PlayerRiderComponentRef:Player Rider Component Object Reference) -> (Output_Get:Player Rider Component Object Reference)
  ]}
 EmptyEnergyDelta {nodes [
   EmptyEnergyDelta() {is=title K2Node_FunctionEntry}
   > n1 = GetDynamicMaterial(self=Get DeltaBar() -> (DeltaBar:Image Object Reference))
   > SetScalarParameterValue(self=@n1, ParameterName=Progress, Value=0)
  ]}
 EventGraph {nodes [
   n1 = RiderUIEnergyChanged!() -> (OutputDelegate, then, Energy:double=~) {is=title K2Node_CustomEvent}
   > UpdateEnergy!(execute, self, Value:double=@n1.Energy) -> (then)
   > UpdateBoostEffect!(execute, self, ProgressRatio:double=@n1.Energy) -> (then)
   n2 = ListenToSprintEvents!() -> (OutputDelegate, then, ListenForSprinting:bool=~) {is=title K2Node_CustomEvent}
   > n3 = Branch(Condition=@n2.ListenForSprinting)
   > n4 = Get!(execute) -> (then, else, WaitingForSprint:Ability Async Wait Gameplay Tag Added Object Reference)
   > n5 = Branch(Condition=IsActive!(self:Cancellable Async Action Object Reference=@n4) -> (ReturnValue:bool))
   n6 = Cast To FortPawn!(execute=[@n5.else, @n4.else], Object=GetOwningPlayerPawn!(self:User Widget Object Reference) -> (ReturnValue:Pawn Object Reference)) -> (then, CastFailed, AsFort Pawn:Fort Pawn Object Reference) {is=title K2Node_DynamicCast}
   > n7 = WaitGameplayTagAddToActor!(
    execute:Exec
    TargetActor:Actor Object Reference=@n6
    Tag:Gameplay Tag Structure=Get SprintingAbilityTag() -> (SprintingAbilityTag:Gameplay Tag Structure)
    OnlyTriggerOnce:bool=true
   ) -> (then:Exec, AsyncTaskProxy:Ability Async Wait Gameplay Tag Added Object Reference, Added:Exec) {is=title K2Node_AsyncAction}
   > Set WaitingForSprint(WaitingForSprint:Ability Async Wait Gameplay Tag Added Object Reference=@n7) -> (Output_Get:Ability Async Wait Gameplay Tag Added Object Reference)
   OnSprintStarted(execute=@n7.Added)
   > ListenToSprintEvents!(execute, self, ListenForSprinting:bool=false) -> (then) {is=title K2Node_CallFunction}
   n8 = Get!(execute=@n3.else) -> (then, else, WaitingForSprintStopped:Ability Async Wait Gameplay Tag Removed Object Reference)
   > n9 = Branch(Condition=IsActive(self=@n8))
   n10 = Cast To FortPawn(execute=[@n9.else, @n8.else], Object=GetOwningPlayerPawn())
   > n11 = WaitGameplayTagRemoveFromActor!(execute:Exec, TargetActor:Actor Object Reference=@n10, Tag:Gameplay Tag Structure=Get SprintingAbilityTag(), OnlyTriggerOnce:bool=true) -> (then:Exec, AsyncTaskProxy:Ability Async Wait Gameplay Tag Removed Object Reference, Removed:Exec) {is=title K2Node_AsyncAction}
   > Set WaitingForSprintStopped(WaitingForSprintStopped:Ability Async Wait Gameplay Tag Removed Object Reference=@n11) -> (Output_Get:Ability Async Wait Gameplay Tag Removed Object Reference)
   OnSprintStopped(execute=@n11.Removed)
   > ListenToSprintEvents(ListenForSprinting=true)
   n12 = InitRiderWidget() {is=title K2Node_CustomEvent}
   > SetupEventBinds()
   n13 = RiderUIHealthChanged!() -> (OutputDelegate, then, Health:double=~) {is=title K2Node_CustomEvent}
   > UpdateHealth!(execute, self, Value:double=@n13.Health) -> (then)
   n14 = OnRiderUINewStats!() -> (OutputDelegate, then, Name:Text, Icon:Slate Brush Structure, RidableActor:Actor Object Reference) {is=title K2Node_CustomEvent}
   > SetVisibility(InVisibility=HitTestInvisible)
   > SetBrush(self=Get SprintIcon() -> (SprintIcon:Image Object Reference), InBrush=@n14.Icon)
   > n15 = Cast To FortPawn(Object=@n14.RidableActor)
   > Set AIPawn(AIPawn:Fort Pawn Object Reference=@n15) -> (Output_Get:Fort Pawn Object Reference)
   n16 = GetMaxHealth!(self:Fort Pawn Object Reference=@n15) -> (ReturnValue:float)
   n17 = GetHealth!(self:Fort Pawn Object Reference=@n15) -> (ReturnValue:float)
   n18 = SelectFloat(A=float / float(A=@n17, B=@n16), B=@n16, bPickA="float > float"(A=@n17))
   > Set PrevHealth(PrevHealth:double=@n18) -> (Output_Get:double)
   > UpdateHealth(Value=@n18)
   n19 = GetComponentByClass!(
    self:Actor Object Reference=@n14.RidableActor
    ComponentClass:Actor Component Class Reference=/RidingContent/Components/CreatureBaseRidableComponent.CreatureBaseRidableComponent_C
   ) -> (ReturnValue:Creature Base Ridable Component Object Reference)
   > Is Valid(InputObject=@n19)
   n20 = Get currentEnergyValue!(self:Creature Base Ridable Component Object Reference=@n19) -> (currentEnergyValue:double)
   > UpdateEnergy(Value=@n20)
   > UpdateBoostEffect(ProgressRatio=@n20)
   n21 = Get A_BoostCompletedAnim() -> (A_BoostCompletedAnim:Widget Animation Object Reference)
   > PlayAnimation(InAnimation=@n21, StartAtTime=GetEndTime(self=@n21))
   > SetVisibility(
    self=[
     Get Overlay_BoostSparks() -> (Overlay_BoostSparks:Overlay Object Reference)
     Get BoostGlowBorder() -> (BoostGlowBorder:Border Object Reference)
    ]
    InVisibility=Collapsed
   )
   OnSprintStarted() {is=title K2Node_CustomEvent}
   > SetBoostAnim!(execute, self, IsBoosting:bool=true) -> (then)
   n22 = OnRiderUIEnd!() -> (OutputDelegate, then, RidableActor:Actor Object Reference) {is=title K2Node_CustomEvent}
   > Branch(Condition="Equal (Object)"(A=@n22.RidableActor, B=Get AIPawn() -> (AIPawn:Fort Pawn Object Reference)))
   > SetVisibility(InVisibility=Collapsed)
   OnSprintStopped() {is=title K2Node_CustomEvent}
   > SetBoostAnim(IsBoosting=false)
   n23 = SetBoostAnim!() -> (OutputDelegate, then, IsBoosting:bool=~) {is=title K2Node_CustomEvent}
   > n24 = Set IsBoostingState(IsBoostingState:bool=@n23.IsBoosting) -> (Output_Get:bool)
   > n25 = Branch(Condition=@n24)
   > StopAnimation(InAnimation=Get A_BoostCompletedAnim())
   > PlayAnimation(InAnimation=Get A_BoostStartedAnim() -> (A_BoostStartedAnim:Widget Animation Object Reference))
   n26 = Get BoostGlowBorder()
   n27 = Get Overlay_BoostSparks()
   > SetVisibility(self=[@n26, @n27], InVisibility=SelfHitTestInvisible)
   StopAnimation(execute=@n25.else, InAnimation=Get A_BoostStartedAnim())
   > n28 = "Play Animation with Finished event (legacy)"!(
    execute:Exec
    Widget:User Widget Object Reference=Self-Reference()
    InAnimation:Widget Animation Object Reference=Get A_BoostCompletedAnim()
    StartAtTime:float=0
    NumLoopsToPlay:int=1
    PlayMode:EUMGSequencePlayMode Enum
    PlaybackSpeed:float=1
   ) -> (then:Exec, Result:UMGSequence Player Object Reference, Finished:Exec) {is=title K2Node_PlayAnimation}
   n29 = Branch(execute=@n28.Finished, Condition=Get IsBoostingState() -> (IsBoostingState:bool))
   SetVisibility(execute=@n29.else, self=[@n26, @n27], InVisibility=Collapsed)
   Event Destruct()
   > ClearOldBindings() {is=title K2Node_CallFunction}
   Event Construct()
   n30 = GetOwningPlayer()
   > n31 = Branch(Condition=AND Boolean(A=IsMobileGame(), B=IsUsingTouch(PlayerController=@n30)))
   > n32 = MobileForceRefresh()
   SetupEventBinds(execute=[@n33.then, @n32.then])
   > Bind Event to On Pawn Set(
    self:Fort HUDContext Object Reference=Get FortHUDContext!() -> (ReturnValue:Fort HUDContext Object Reference) {is=title K2Node_GetBlueprintContextEx}
    Delegate=@n12
   ) {is=title K2Node_AddDelegate}
   > ListenToSprintEvents(ListenForSprinting=true)
   > Initialize!(
    execute
    self:Common Tag Visibility Widget Object Reference=Get TagVisibility_HealthValue() -> (TagVisibility_HealthValue:Common Tag Visibility Widget Object Reference)
    OwningLocalPlayer:Local Player Object Reference=GetOwningLocalPlayer!(self:Widget Object Reference) -> (ReturnValue:Local Player Object Reference)
    OwningPlayerController:Player Controller Object Reference=GetOwningPlayer()
   ) -> (then)
   n33 = SetVisibility(execute=@n31.else, InVisibility=Collapsed)
   Event OnInitialized()
   > Branch(Condition=IsUsingTouch(PlayerController=GetOwningPlayer()))
   n34 = Get SizeBox_Progress() -> (SizeBox_Progress:Size Box Object Reference)
   > SetHeightOverride(self=@n34, InHeightOverride=15)
   > SetWidthOverride(self=@n34, InWidthOverride=275)
   > SetWidthOverride(self=Get SizeBox_Health() -> (SizeBox_Health:Size Box Object Reference), InWidthOverride=275)
   > SetPadding!(
    execute
    self:Horizontal Box Slot Object Reference=SlotAsHorizontalBoxSlot!(Widget:Widget Object Reference=Get SprintIcon()) -> (ReturnValue:Horizontal Box Slot Object Reference)
    InPadding_Left:float=0
    InPadding_Top:float=0
    InPadding_Right:float=3
    InPadding_Bottom:float=0
   ) -> (then)
  ]}
 MobileForceRefresh {nodes [
   MobileForceRefresh() {is=title K2Node_FunctionEntry}
   n1 = Get CachedRidableActor!(self:Player Rider Component Object Reference=GetComponentByClass(self=GetOwningPlayerPawn(), ComponentClass=/RidingContent/Components/PlayerRiderComponent.PlayerRiderComponent_C) -> (ReturnValue:Player Rider Component Object Reference)) -> (CachedRidableActor:Actor Object Reference)
   > Is Valid(InputObject=@n1)
   n2 = GetComponentByClass(self=@n1, ComponentClass=&CreatureBaseRidableComponent_C) -> (ReturnValue:Creature Base Ridable Component Object Reference)
   > Is Valid(InputObject=@n2)
   > n3 = Cast To FortPawn(Object=@n1)
   n4 = GetHealth(self=@n3)
   > UpdateHealth(Value=@n4)
   > SetHealthValue!(execute, self, Value:double=@n4) -> (then)
   n5 = Get currentEnergyValue(self=@n2)
   > UpdateEnergy(Value=@n5)
   > UpdateBoostEffect(ProgressRatio=@n5)
   > SetVisibility(InVisibility=Visible)
  ]}
 OnOvershieldVisibilityChanged {nodes [
   n1 = OnOvershieldVisibilityChanged!() -> (then, IsVisible:bool=~) {is=title K2Node_FunctionEntry}
   n2 = GetOwningPlayer()
   > Branch(Condition=NOT Boolean(A=IsUsingTouch(PlayerController=@n2)))
   > SetRenderTranslation(Translation=MakeVector2D(X=0, Y=SelectFloat(A=0, B=15, bPickA=@n1)))
  ]}
 Set Delta {nodes [
   n1 = Set Delta!() -> (
    then
    Target:Image Object Reference
    CurrentValue:double=~
    PreviousValue:double=~
    Threshold:double=~
    FunctionName:String
    TimerHandle:"Timer Handle Structure (by ref)"
    DeltaDecayDelay:double=~
   ) {is=title K2Node_FunctionEntry}
   > n2 = Branch(Condition="float <= float"(A=float - float(A=@n1.CurrentValue, B=@n1.PreviousValue), B=@n1.Threshold) -> (ReturnValue=~))
   > Clear and Invalidate Timer by Handle(Handle=@n1.TimerHandle)
   > Set CooldownStartTime(CooldownStartTime:double=GetServerWorldTimeSeconds!() -> (ReturnValue:float)) -> (Output_Get:double)
   > n3 = Set Timer by Function Name!(
    execute
    Object:Object Reference
    FunctionName:String=@n1.FunctionName
    Time:float=@n1.DeltaDecayDelay
    bLooping:bool=true
    bMaxOncePerFrame:bool=false
    InitialStartDelay:float=0.33
    InitialStartDelayVariance:float=0
   ) -> (then, ReturnValue:Timer Handle Structure) {is=title K2Node_CallFunction}
   > Return Node!(execute, TimerHandleRef:Timer Handle Structure=@n3)
   n4 = Branch(execute=@n2.else, Condition=Is Timer Active by Handle!(Handle:Timer Handle Structure=@n1.TimerHandle) -> (ReturnValue:bool))
   > Return Node()
   n5 = Set Timer by Function Name(
    execute=@n4.else
    FunctionName=@n1.FunctionName
    Time=0.01
    bLooping=false
    bMaxOncePerFrame=false
    InitialStartDelay=0
    InitialStartDelayVariance=0
   )
   > Return Node(TimerHandleRef=@n5)
  ]}
 SetHealthValue {nodes [
   n1 = SetHealthValue!() -> (then, Value:double=~) {is=title K2Node_FunctionEntry}
   n2 = Round(A=@n1) {is=title K2Node_CallFunction}
   > n3 = Branch(Condition=AND Boolean(A="Equal (Integer)"(A=@n2, B=0) {is=title K2Node_CallFunction}, B="float > float"(A=@n1, B=0) {is=title K2Node_CallFunction}))
   n4 = Get CurrentValue() -> (CurrentValue:Common Text Object Reference)
   > "SetText (Text)"(self=@n4, InText="To Text (Integer)"(Value=1))
   "SetText (Text)"(execute=@n3.else, self=@n4, InText="To Text (Integer)"(Value=@n2))
  ]}
 SetProgressBar {nodes [
   n1 = SetProgressBar!() -> (
    then
    Value:double=~
    TargetFillBar:Image Object Reference
    DeltaBar:Image Object Reference
    PreviousValue:double=~
    Threshold:double=~
    FunctionName:String
    TimerHandle:"Timer Handle Structure (by ref)"
    DeltaDecayDelay:double=~
   ) {is=title K2Node_FunctionEntry}
   > n2 = Sequence()
   > n3 = GetDynamicMaterial(self=@n1.TargetFillBar)
   > SetScalarParameterValue(self=@n3, ParameterName=Progress, Value=@n1)
   > SetRenderOpacity(self=Get EmptyStatusGlow() -> (EmptyStatusGlow:Image Object Reference), InOpacity=Lerp(A=1, B=-0.5, Alpha=@n1))
   > n4 = GetDynamicMaterial(self=Get EmptyStatusGlow())
   > SetVectorParameterValue(self=@n4, ParameterName=Progress_Color_Start, Value="Lerp (LinearColor)"(A=(R=1,G=0.3333,B=0,A=1), B=(R=1,G=0.1225,B=0.025,A=1), Alpha=@n1))
   n5 = Set Delta!(
    execute=@n2.then_1
    self
    Target:Image Object Reference=@n1.DeltaBar
    CurrentValue:double=@n1
    PreviousValue:double=@n1.PreviousValue
    Threshold:double=@n1.Threshold
    FunctionName:String=@n1.FunctionName
    TimerHandle:"Timer Handle Structure (by ref)"=@n1.TimerHandle
    DeltaDecayDelay:double=@n1.DeltaDecayDelay
   ) -> (then, TimerHandleRef:Timer Handle Structure) {is=title K2Node_CallFunction}
   > Return Node!(execute, DeltaTimerHandle:"Timer Handle Structure (by ref)"=@n5)
  ]}
 SetupEventBinds {nodes [
   SetupEventBinds() {is=title K2Node_FunctionEntry}
   > ClearOldBindings()
   n1 = GetOwningPlayerPawn()
   > Is Valid(InputObject=@n1)
   > n2 = Set PlayerRiderComponentRef(PlayerRiderComponentRef=GetComponentByClass(self=@n1, ComponentClass=&PlayerRiderComponent_C) -> (ReturnValue:Player Rider Component Object Reference))
   > Is Valid(InputObject=@n2)
   > Bind Event to Rider UIEnd(self:Player Rider Component Object Reference=@n2, Delegate=Create Event()) {is=title K2Node_AddDelegate}
   > Bind Event to Rider UINew Stats(self:Player Rider Component Object Reference=@n2, Delegate=Create Event()) {is=title K2Node_AddDelegate}
   > Bind Event to Rider UIUpdate Energy(self:Player Rider Component Object Reference=@n2, Delegate=Create Event()) {is=title K2Node_AddDelegate}
   > Bind Event to Rider UIUpdate Health(self:Player Rider Component Object Reference=@n2, Delegate=Create Event()) {is=title K2Node_AddDelegate}
   n3 = GetPlayerStateZone!(self:Fort HUDContext Object Reference=Get FortHUDContext()) -> (ReturnValue:Fort Player State Zone Object Reference)
   > Is Valid(InputObject=@n3)
   > Bind Event to On Overshield Visibility Changed(self:Fort Player State Zone Object Reference=@n3, Delegate=Create Event()) {is=title K2Node_AddDelegate}
   n4 = Get FortHUDContext()
   > OnOvershieldVisibilityChanged!(
    execute
    self
    IsVisible:bool=ShouldOvershieldBarBeVisible!(self:Fort Player State Zone Object Reference=GetPlayerStateZone(self=@n4)) -> (ReturnValue:bool)
   ) -> (then) {is=title K2Node_CallFunction}
  ]}
 TryPlayingLowState {nodes [
   n1 = TryPlayingLowState!() -> (then, InAnimation:Widget Animation Object Reference) {is=title K2Node_FunctionEntry}
   > n2 = Branch(Condition=IsAnimationPlaying(InAnimation=@n1))
   PlayAnimation(execute=@n2.else, InAnimation=@n1)
  ]}
 UpdateBoostEffect {nodes [
   n1 = UpdateBoostEffect!() -> (then, ProgressRatio:double=~) {is=title K2Node_FunctionEntry}
   > SetRenderTranslation(self=Get Overlay_BoostSparks(), Translation=MakeVector2D(X=Lerp(A=-328, B=-1, Alpha=@n1), Y=0))
   > SetRenderScale!(execute, self:Widget Object Reference=Get BoostGlowBorder(), Scale:Vector 2D Structure=MakeVector2D(X=Lerp(A=0, B=1, Alpha=@n1), Y=1)) -> (then)
  ]}
 UpdateEnergy {nodes [
   n1 = UpdateEnergy!() -> (then, Value:double=~) {is=title K2Node_FunctionEntry}
   > n2 = SetProgressBar!(
    execute
    self
    Value:double=@n1
    TargetFillBar:Image Object Reference=Get FillBar() -> (FillBar:Image Object Reference)
    DeltaBar:Image Object Reference=Get DeltaBar()
    PreviousValue:double=Get PrevEnergy() -> (PrevEnergy:double)
    Threshold:double=Get Anim_DeltaThreshold() -> (Anim_DeltaThreshold:double)
    FunctionName:String=EmptyEnergyDelta
    TimerHandle:"Timer Handle Structure (by ref)"=Get EmptyEnergyDelta_Handle() -> (EmptyEnergyDelta_Handle:Timer Handle Structure)
    DeltaDecayDelay:double=Get EnergyDeltaDecayDelay() -> (EnergyDeltaDecayDelay:double)
   ) -> (then, DeltaTimerHandle:Timer Handle Structure) {is=title K2Node_CallFunction}
   > Set EmptyEnergyDelta_Handle(EmptyEnergyDelta_Handle:Timer Handle Structure=@n2) -> (Output_Get:Timer Handle Structure)
   > Set PrevEnergy(PrevEnergy:double=@n1) -> (Output_Get:double)
   > Branch(Condition="float <= float"(A=@n1, B=Get LowEnergyThreshold() -> (LowEnergyThreshold:double)) -> (ReturnValue=~))
   > TryPlayingLowState!(execute, self, InAnimation:Widget Animation Object Reference=Get A_LowAnim() -> (A_LowAnim:Widget Animation Object Reference)) -> (then) {is=title K2Node_CallFunction}
  ]}
 UpdateHealth {nodes [
   n1 = UpdateHealth!() -> (then, Value:double=~) {is=title K2Node_FunctionEntry}
   > PrintText!(
    execute
    InText=Format Text!(Format="LOCTEXT(\"RiderWidget-UpdateHealth value {0}\")", 0:double=@n1) -> (Result)
    bPrintToScreen
    bPrintToLog
    TextColor
    Duration
    Key
   ) -> (then)
   > n2 = SetProgressBar(
    Value=@n1
    TargetFillBar=Get FillBar_1() -> (FillBar_1:Image Object Reference)
    DeltaBar=Get DeltaBar_1() -> (DeltaBar_1:Image Object Reference)
    PreviousValue=Get PrevHealth() -> (PrevHealth:double)
    Threshold=Get Anim_DeltaThreshold()
    FunctionName=UpdateHealthDelta
    TimerHandle=Get EmptyHealthDelta_Handle() -> (EmptyHealthDelta_Handle:Timer Handle Structure)
    DeltaDecayDelay=Get HealthDeltaDecayDelay() -> (HealthDeltaDecayDelay:double)
   )
   > Set EmptyHealthDelta_Handle(EmptyHealthDelta_Handle:Timer Handle Structure=@n2) -> (Output_Get:Timer Handle Structure)
   > Set PrevHealth(PrevHealth=@n1)
   > n3 = Get!(execute) -> (then, else, AIPawn:Fort Pawn Object Reference)
   > SetHealthValue(Value=float * float(A=GetMaxHealth(self=@n3), B=Get Value() -> (Value:double)) {is=title K2Node_PromotableOperator}) {is=title K2Node_CallFunction}
  ]}
 UpdateHealthDelta {nodes [
   UpdateHealthDelta() {is=title K2Node_FunctionEntry}
   > n1 = GetDynamicMaterial(self=Get DeltaBar_1())
   > n2 = Get Scalar Parameter Value!(execute, self:Material Instance Dynamic Object Reference=@n1, ParameterName:Name=Progress) -> (then, ReturnValue:float) {is=title K2Node_CallFunction}
   n3 = Get PrevHealth()
   n4 = FInterpEaseInOut!(
    A:double=@n2
    B:double=@n3
    Alpha:double=float - float(A=GetServerWorldTimeSeconds(), B=Get CooldownStartTime() -> (CooldownStartTime:double))
    Exponent:double=0.5
   ) -> (ReturnValue:double)
   > n5 = Branch(Condition="Equal (Float)"(A=@n4, B=@n3))
   > Clear and Invalidate Timer by Handle(Handle=Get EmptyHealthDelta_Handle())
   SetScalarParameterValue(execute=@n5.else, self=@n1, ParameterName=Progress, Value=@n4)
  ]}
]
variables [
 AIPawn {type=FortPawn, default=null}
 Anim_DeltaThreshold {type=number}
 CachedCurrentEnergy {type=number}
 CahcedProgressDynamicMat {type=MaterialInstanceDynamic}
 ColorStart {type=LinearColor, default=<rgba 0 0 0 0>}
 CooldownStartTime {type=number}
 EmptyEnergyDelta_Handle {type=TimerHandle}
 EmptyHealthDelta_Handle {type=TimerHandle}
 EnergyDeltaDecayDelay {type=number, default=0.33}
 HealthDeltaDecayDelay {type=number, default=0.01}
 IsBoostingState {type=boolean}
 IsListeningForSprint {type=boolean}
 IsListeningForSprintStopped {type=boolean}
 LowEnergyThreshold {type=number, default=0.1}
 PlayerRiderComponentRef {type=PlayerRiderComponent_C, default=null}
 PrevEnergy {type=number, default=1}
 PrevHealth {type=number}
 SprintingAbilityTag {type=GameplayTag, default {tagName=Gameplay.Action.Npc.RidingSprinting}}
 WaitingForSprint {type=AbilityAsync_WaitGameplayTagAdded, default=null}
 WaitingForSprintStopped {type=AbilityAsync_WaitGameplayTagRemoved, default=null}
]
widgets {
 info {inheritedWidgetCount=1, parentClass=&FortHUDElementWidget, rootWidgetClass=&Overlay, widgetCount=25}
 widgets [
  Overlay_Stats {widgetClassPath=&Overlay, bIsVariable=true, widget=Overlay_Stats}
  HorizontalBox_50 {
   widgetClassPath=&HorizontalBox
   bIsVariable=true
   parent=Overlay_Stats
   widget=HorizontalBox_50
   slot {class=OverlaySlot, properties {verticalAlignment=VAlign_Fill}}
  }
  SprintIcon {
   widgetClassPath=&Image
   bIsVariable=true
   parent=HorizontalBox_50
   widget=SprintIcon
   properties {brush {
     bIsDynamicallyLoaded=false
     drawAs=Image
     imageType=NoImage
     mirroring=NoMirror
     resourceName=None
     resourceObject=/Game/Athena/HUD/Art/T_UI_Wolf_Icon_64x
     tiling=NoTile
     imageSize=<32 32>
     margin=<ltrb 0.2 0.2 0.2 0.2>
     outlineSettings {
      bUseBrushTransparency=false
      roundingType=HalfHeightRadius
      width=0
      color {colorUseRule=UseColor_Specified, specifiedColor=<rgba 0 0 0 0>}
      cornerRadii=<0 0 0 0>
     }
     tintColor {colorUseRule=UseColor_Specified, specifiedColor=<rgba 1 1 1 1>}
     uVRegion {bIsValid=false, max=<0 0>, min=<0 0>}
    }}
   slot {class=HorizontalBoxSlot, properties {horizontalAlignment=HAlign_Left, verticalAlignment=VAlign_Center, padding=<ltrb 0 0 5 0>}}
  }
  VerticalBox_85 {
   widgetClassPath=&VerticalBox
   parent=HorizontalBox_50
   widget=VerticalBox_85
   slot {class=HorizontalBoxSlot, properties {horizontalAlignment=HAlign_Left, verticalAlignment=VAlign_Center}}
  }
  SizeBox_Progress {
   widgetClassPath=&SizeBox
   bIsVariable=true
   parent=VerticalBox_85
   widget=SizeBox_Progress
   properties {bOverride_HeightOverride=true, bOverride_WidthOverride=true, heightOverride=6, widthOverride=333}
   slot {class=VerticalBoxSlot, properties {horizontalAlignment=HAlign_Left, verticalAlignment=VAlign_Top, padding=<ltrb 0 0 0 6>}}
  }
  Overlay_2 {widgetClassPath=&Overlay, parent=SizeBox_Progress, widget=Overlay_2, slot {class=SizeBoxSlot}}
  EmptyStatusGlow {
   widgetClassPath=&Image
   bIsVariable=true
   parent=Overlay_2
   widget=EmptyStatusGlow
   properties {
    renderOpacity=0
    brush {
     bIsDynamicallyLoaded=false
     drawAs=Box
     imageType=NoImage
     mirroring=NoMirror
     resourceName=None
     resourceObject=/Game/Athena/HUD/LTM/Wax/Materials/M_UI_BallerBoostGlow
     tiling=NoTile
     imageSize=<32 32>
     margin=<ltrb 0.5 0.5 0.5 0.5>
     outlineSettings {
      bUseBrushTransparency=false
      roundingType=HalfHeightRadius
      width=0
      color {colorUseRule=UseColor_Specified, specifiedColor=<rgba 0 0 0 0>}
      cornerRadii=<0 0 0 0>
     }
     tintColor {colorUseRule=UseColor_Specified, specifiedColor=<rgba 1 1 1 1>}
     uVRegion {bIsValid=false, max=<0 0>, min=<0 0>}
    }
    renderTransform {angle=0, scale=<-1 1>, shear=<0 0>, translation=<0 0>}
   }
   slot {class=OverlaySlot, properties {horizontalAlignment=HAlign_Fill, verticalAlignment=VAlign_Fill, padding=<ltrb -13 -12 128 -12>}}
  }
  DeltaBar {
   widgetClassPath=&Image
   bIsVariable=true
   parent=Overlay_2
   widget=DeltaBar
   properties {
    visibility=HitTestInvisible
    brush {
     bIsDynamicallyLoaded=false
     drawAs=Image
     imageType=NoImage
     mirroring=NoMirror
     resourceName=None
     resourceObject=/RidingContent/UI/Materials/M_UI_Animal_EnergyDelta
     tiling=NoTile
     imageSize=<32 32>
     margin=<ltrb 0 0 0 0>
     outlineSettings {
      bUseBrushTransparency=false
      roundingType=HalfHeightRadius
      width=0
      color {colorUseRule=UseColor_Specified, specifiedColor=<rgba 0 0 0 0>}
      cornerRadii=<0 0 0 0>
     }
     tintColor {colorUseRule=UseColor_Specified, specifiedColor=<rgba 1 1 1 1>}
     uVRegion {bIsValid=false, max=<0 0>, min=<0 0>}
    }
   }
   slot {class=OverlaySlot, properties {horizontalAlignment=HAlign_Fill}}
  }
  FillBar {
   widgetClassPath=&Image
   bIsVariable=true
   parent=Overlay_2
   widget=FillBar
   properties {brush {
     bIsDynamicallyLoaded=false
     drawAs=Image
     imageType=NoImage
     mirroring=NoMirror
     resourceName=None
     resourceObject=/RidingContent/UI/Materials/M_UI_Animal_Energy
     tiling=NoTile
     imageSize=<32 32>
     margin=<ltrb 0 0 0 0>
     outlineSettings {
      bUseBrushTransparency=false
      roundingType=HalfHeightRadius
      width=0
      color {colorUseRule=UseColor_Specified, specifiedColor=<rgba 0 0 0 0>}
      cornerRadii=<0 0 0 0>
     }
     tintColor {colorUseRule=UseColor_Specified, specifiedColor=<rgba 1 1 1 1>}
     uVRegion {bIsValid=false, max=<0 0>, min=<0 0>}
    }}
   slot {class=OverlaySlot, properties {horizontalAlignment=HAlign_Fill, verticalAlignment=VAlign_Center}}
  }
  BoostGlowBorder {
   widgetClassPath=&Border
   bIsVariable=true
   parent=Overlay_2
   widget=BoostGlowBorder
   properties {
    visibility=Collapsed
    background {
     bIsDynamicallyLoaded=false
     drawAs=NoDrawType
     imageType=NoImage
     mirroring=NoMirror
     resourceName=None
     resourceObject=null
     tiling=NoTile
     imageSize=<32 32>
     margin=<ltrb 0 0 0 0>
     outlineSettings {
      bUseBrushTransparency=false
      roundingType=HalfHeightRadius
      width=0
      color {colorUseRule=UseColor_Specified, specifiedColor=<rgba 0 0 0 0>}
      cornerRadii=<0 0 0 0>
     }
     tintColor {colorUseRule=UseColor_Specified, specifiedColor=<rgba 1 1 1 1>}
     uVRegion {bIsValid=false, max=<0 0>, min=<0 0>}
    }
    desiredSizeScale=<0.5 1>
    padding=<ltrb 0 0 0 0>
    renderTransformPivot=<0 0.5>
   }
   slot {class=OverlaySlot, properties {horizontalAlignment=HAlign_Fill, verticalAlignment=VAlign_Fill}}
  }
  SizeBox_1 {
   widgetClassPath=&SizeBox
   parent=BoostGlowBorder
   widget=SizeBox_1
   properties {bOverride_WidthOverride=true, widthOverride=300}
   slot {class=BorderSlot, properties {padding=<ltrb 0 0 0 0>}}
  }
  BoostGlow {
   widgetClassPath=&Image
   bIsVariable=true
   parent=SizeBox_1
   widget=BoostGlow
   properties {
    brush {
     bIsDynamicallyLoaded=false
     drawAs=Box
     imageType=NoImage
     mirroring=NoMirror
     resourceName=None
     resourceObject=&M_UI_BallerBoostGlow
     tiling=NoTile
     imageSize=<72 72>
     margin=<ltrb 0.5 0.5 0.5 0.5>
     outlineSettings {
      bUseBrushTransparency=false
      roundingType=HalfHeightRadius
      width=0
      color {colorUseRule=UseColor_Specified, specifiedColor=<rgba 0 0 0 0>}
      cornerRadii=<0 0 0 0>
     }
     tintColor {colorUseRule=UseColor_Specified, specifiedColor=<rgba 1 1 1 1>}
     uVRegion {bIsValid=false, max=<0 0>, min=<0 0>}
    }
    renderTransformPivot=<0 0.5>
   }
   slot {class=SizeBoxSlot, properties {padding=<ltrb -32 -32 -32 -32>}}
  }
  Overlay_BoostSparks {
   widgetClassPath=&Overlay
   bIsVariable=true
   parent=Overlay_2
   widget=Overlay_BoostSparks
   properties {visibility=Collapsed, renderTransform {angle=0, scale=<1 1>, shear=<0 0>, translation=<-1 0>}, renderTransformPivot=<0 0.5>}
   slot {class=OverlaySlot, properties {horizontalAlignment=HAlign_Right, verticalAlignment=VAlign_Center}}
  }
  Overlay_47 {
   widgetClassPath=&Overlay
   parent=Overlay_BoostSparks
   widget=Overlay_47
   properties {visibility=HitTestInvisible, renderTransform {angle=-90, scale=<1 1>, shear=<0 0>, translation=<0 0>}}
   slot {class=OverlaySlot}
  }
  Image_Sparks {
   widgetClassPath=&Image
   bIsVariable=true
   parent=Overlay_47
   widget=Image_Sparks
   properties {
    visibility=HitTestInvisible
    brush {
     bIsDynamicallyLoaded=false
     drawAs=Image
     imageType=NoImage
     mirroring=NoMirror
     resourceName=None
     resourceObject=/Game/Athena/UI/InGame/Interrogation/Materials/M_UI_Starburst_Sparks
     tiling=NoTile
     imageSize=<295.3884 400>
     margin=<ltrb 0 0 0 0>
     outlineSettings {
      bUseBrushTransparency=false
      roundingType=HalfHeightRadius
      width=0
      color {colorUseRule=UseColor_Specified, specifiedColor=<rgba 0 0 0 0>}
      cornerRadii=<0 0 0 0>
     }
     tintColor {colorUseRule=UseColor_Specified, specifiedColor=<rgba 1 0.3682 0.0743 1>}
     uVRegion {bIsValid=false, max=<0 0>, min=<0 0>}
    }
    renderTransform {angle=-12, scale=<1 1>, shear=<0 0>, translation=<0 0>}
   }
   slot {class=OverlaySlot, properties {horizontalAlignment=HAlign_Center, verticalAlignment=VAlign_Center, padding=<ltrb -200 -200 -200 -200>}}
  }
  Image_Sparks_1 {
   widgetClassPath=&Image
   bIsVariable=true
   parent=Overlay_47
   widget=Image_Sparks_1
   properties {
    visibility=HitTestInvisible
    brush {
     bIsDynamicallyLoaded=false
     drawAs=Image
     imageType=NoImage
     mirroring=NoMirror
     resourceName=None
     resourceObject=/Game/Athena/UI/InGame/Interrogation/Materials/M_UI_Starburst_Sparks_Inst
     tiling=NoTile
     imageSize=<120 400>
     margin=<ltrb 0 0 0 0>
     outlineSettings {
      bUseBrushTransparency=false
      roundingType=HalfHeightRadius
      width=0
      color {colorUseRule=UseColor_Specified, specifiedColor=<rgba 0 0 0 0>}
      cornerRadii=<0 0 0 0>
     }
     tintColor {colorUseRule=UseColor_Specified, specifiedColor=<rgba 1 0.7183 0.651 1>}
     uVRegion {bIsValid=false, max=<0 0>, min=<0 0>}
    }
    renderTransform {angle=-12, scale=<1 1>, shear=<0 0>, translation=<0 0>}
   }
   slot {class=OverlaySlot, properties {horizontalAlignment=HAlign_Center, verticalAlignment=VAlign_Center, padding=<ltrb -200 -200 -200 -200>}}
  }
  Image_LargeCone {
   widgetClassPath=&Image
   bIsVariable=true
   parent=Overlay_47
   widget=Image_LargeCone
   properties {
    visibility=HitTestInvisible
    brush {
     bIsDynamicallyLoaded=false
     drawAs=Image
     imageType=NoImage
     mirroring=NoMirror
     resourceName=None
     resourceObject=/Game/Athena/HUD/Vehicles/Materials/M_UI_SparksGlow_Inst
     tiling=NoTile
     imageSize=<150 200>
     margin=<ltrb 0 0 0 0>
     outlineSettings {
      bUseBrushTransparency=false
      roundingType=HalfHeightRadius
      width=0
      color {colorUseRule=UseColor_Specified, specifiedColor=<rgba 0 0 0 0>}
      cornerRadii=<0 0 0 0>
     }
     tintColor {colorUseRule=UseColor_Specified, specifiedColor=<rgba 1 1 1 1>}
     uVRegion {bIsValid=false, max=<0 0>, min=<0 0>}
    }
    renderTransform {angle=-12, scale=<1 1>, shear=<0 0>, translation=<0 28>}
   }
   slot {class=OverlaySlot, properties {horizontalAlignment=HAlign_Center, verticalAlignment=VAlign_Center, padding=<ltrb -200 -200 -200 -200>}}
  }
  HorizontalBox_0 {
   widgetClassPath=&HorizontalBox
   parent=VerticalBox_85
   widget=HorizontalBox_0
   slot {class=VerticalBoxSlot, properties {verticalAlignment=VAlign_Bottom, size {sizeRule=Fill, value=1}}}
  }
  SizeBox_Health {
   widgetClassPath=&SizeBox
   bIsVariable=true
   parent=HorizontalBox_0
   widget=SizeBox_Health
   properties {bOverride_HeightOverride=true, bOverride_WidthOverride=true, heightOverride=24, widthOverride=333}
   slot {class=HorizontalBoxSlot}
  }
  Overlay {widgetClassPath=&Overlay, parent=SizeBox_Health, widget=Overlay, slot {class=SizeBoxSlot}}
  DeltaBar_1 {
   widgetClassPath=&Image
   bIsVariable=true
   parent=Overlay
   widget=DeltaBar_1
   properties {
    visibility=HitTestInvisible
    brush {
     bIsDynamicallyLoaded=false
     drawAs=Image
     imageType=NoImage
     mirroring=NoMirror
     resourceName=None
     resourceObject=/RidingContent/UI/Materials/M_UI_Animal_HealthDelta
     tiling=NoTile
     imageSize=<32 32>
     margin=<ltrb 0 0 0 0>
     outlineSettings {
      bUseBrushTransparency=false
      roundingType=HalfHeightRadius
      width=0
      color {colorUseRule=UseColor_Specified, specifiedColor=<rgba 0 0 0 0>}
      cornerRadii=<0 0 0 0>
     }
     tintColor {colorUseRule=UseColor_Specified, specifiedColor=<rgba 1 1 1 1>}
     uVRegion {bIsValid=false, max=<0 0>, min=<0 0>}
    }
   }
   slot {class=OverlaySlot, properties {horizontalAlignment=HAlign_Fill}}
  }
  FillBar_1 {
   widgetClassPath=&Image
   bIsVariable=true
   parent=Overlay
   widget=FillBar_1
   properties {brush {
     bIsDynamicallyLoaded=false
     drawAs=Image
     imageType=NoImage
     mirroring=NoMirror
     resourceName=None
     resourceObject=/RidingContent/UI/Materials/M_UI_Animal_Health
     tiling=NoTile
     imageSize=<32 32>
     margin=<ltrb 0 0 0 0>
     outlineSettings {
      bUseBrushTransparency=false
      roundingType=HalfHeightRadius
      width=0
      color {colorUseRule=UseColor_Specified, specifiedColor=<rgba 0 0 0 0>}
      cornerRadii=<0 0 0 0>
     }
     tintColor {colorUseRule=UseColor_Specified, specifiedColor=<rgba 1 1 1 1>}
     uVRegion {bIsValid=false, max=<0 0>, min=<0 0>}
    }}
   slot {class=OverlaySlot, properties {horizontalAlignment=HAlign_Fill, verticalAlignment=VAlign_Center}}
  }
  TagVisibility_HealthValue {
   widgetClassPath=/Script/FortniteUI.CommonTagVisibilityWidget
   bIsVariable=true
   parent=HorizontalBox_0
   widget=TagVisibility_HealthValue
   properties {hUDElementTag {gameplayTags [{tagName=HUD.BottomLeft.HitPointValues}]}}
   slot {class=HorizontalBoxSlot, properties {horizontalAlignment=HAlign_Left, verticalAlignment=VAlign_Center, padding=<ltrb 7 0 0 0>}}
  }
  CurrentValue {
   widgetClassPath=&CommonTextBlock
   bIsVariable=true
   parent=TagVisibility_HealthValue
   widget=CurrentValue
   properties {
    justification=Right
    lineHeightPercentage=0.8
    text=999?
    visibility=HitTestInvisible
    font {
     bForceMonospaced=false
     bMaterialIsStencil=false
     fontMaterial=null
     fontObject=&BurbankSmall
     letterSpacing=0
     monospacedWidth=1
     size=15
     skewAmount=0
     typefaceFontName=Black
     outlineSettings {
      bApplyOutlineToDropShadows=true
      bMiteredCorners=false
      bSeparateFillAlpha=true
      outlineMaterial=&M-FontShadow
      outlineSize=2
      outlineColor=<rgba 0 0 0 0.5>
     }
    }
    margin=<ltrb 0 -6 0 1>
    renderTransformPivot=<0 0.75>
    shadowOffset=<0 0>
   }
   slot {class=PanelSlot}
  }
  TutorialHighlight {widgetClassPath=&TutorialHighlight}
 ]
}
